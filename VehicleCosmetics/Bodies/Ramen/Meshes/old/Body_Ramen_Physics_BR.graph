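# Physics asset for the Ramen vehicle body cosmetic, named as a BR variant and stored in an "old" meshes folder.
# It holds a single body on the root bone, with no collision shapes listed.
# It has no constraints, so it defines no articulated or jointed parts.
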PhysicsAsset /VehicleCosmetics/Bodies/Ramen/Meshes/old/Body_Ramen_Physics_BR
bodies {root []}
constraints []
